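# Control Rig function library that drives corrective deform bones on the large male player body, covering areas such as upper arms, wrists, hips, knees and elbows.
# Parent constraints blend each deform base between neighbouring bones.
# Spherical pose readers or remapped joint yaw deltas weight offsets from a value struct, mirrored for the right side.
# Curve-struct flags optionally publish those alphas as root-bone animation attributes.
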ControlRigBlueprint /Game/Characters/Player/Male/Large/Bodies/M_LRG_Base/Meshes/DeformationRig/Deform_FNC/FN_DeformRig_ControlRig_FNC
graphs [
 ContainedGraph {nodes [
   n1 = RigVMFunctionEntryNode()
   n2 = &RigUnit_GetTransform(Space=LocalSpace, Item=@n1.Item)
   n3 = &RigUnit_GetTransform(Space=LocalSpace, bInitial=true, Item=@n1.Item)
   > RigVMFunctionReturnNode(OutTransform=/Script/RigVM.RigVMFunction_MathTransformToEulerTransform(Value=&RigVMFunction_MathTransformMakeRelative(Global=@n2.Transform, Parent=@n3.Transform).Local) {is=struct RigVMUnitNode, template="Cast::Execute(in Value,out Result)"}.Result)
  ]}
 ContainedGraph {nodes [
   n1 = RigVMFunctionEntryNode()
   > RigVMVariableNode(Value=@n1.value_struct) {variable=value_struct_var}
   > RigVMVariableNode(Value=@n1.curve_struct) {variable=curve_struct_var}
   > SequenceExecution_2 = RigVMAggregateNode()
   RigVMFunctionReturnNode(ExecuteContext=@SequenceExecution_2.D)
   SequenceExecution_3_1 = RigVMAggregateNode(ExecuteContext=@SequenceExecution_2.B)
   &RigUnit_ParentConstraint(
    AdvancedSettings=(RotationOrderForFilter=XZY)
    Child=(Type=Bone,Name="deform_upperarm_fr_base_l")
    Parents=((Item=(Type=Bone,Name="upperarm_twist_01_l"),Weight=0.3),(Item=(Type=Bone,Name="clavicle_l"),Weight=0.7))
    Weight=1
    ExecutePin=@SequenceExecution_3_1.A
   )
   > n2 = &RigUnit_SphericalPoseReader(
    ActiveRegionScaleFactors=(PositiveWidth=0.5,NegativeWidth=3,PositiveHeight=1.5,NegativeHeight=1.5)
    ActiveRegionSize=0.1
    Debug=(bDrawDebug=false,DebugScale=25,DebugSegments=20,DebugThickness=0.25)
    DriverAxis=(X=1,Y=0,Z=0)
    DriverItem=(Type=Bone,Name="upperarm_l")
    FalloffRegionScaleFactors=(PositiveWidth=1,NegativeWidth=1,PositiveHeight=1,NegativeHeight=1)
    FalloffSize=0.4
    OptionalParentItem=(Type=Bone)
    RotationOffset=(X=90,Y=0,Z=0)
    Debug.bDrawDebug=@n1.debug
   )
   n3 = RigVMVariableNode() {variable=value_struct_var}
   > &RigUnit_ModifyTransforms(
    ItemToModify=((Item=(Type=Bone,Name="deform_upperarm_fr_l"),Transform=(Rotation=(X=0,Y=0,Z=0,W=1),Translation=(X=0,Y=0,Z=0),Scale3D=(X=1,Y=1,Z=1))))
    ItemToModify.0.Transform=@n3.Value.upperarm_fwd_fr_19_#1
    Weight=@n2.OutputParam
   )
   n4 = RigVMVariableNode() {variable=value_struct_var}
   > n5 = &RigUnit_ModifyTransforms(
    ItemToModify=((Item=(Type=Bone,Name="deform_upperarm_bk_l"),Transform=(Rotation=(X=0,Y=0,Z=0,W=1),Translation=(X=0,Y=0,Z=0),Scale3D=(X=1,Y=1,Z=1))))
    ItemToModify.0.Transform=@n4.Value.upperarm_fwd_bk_20_#2
    Weight=@n2.OutputParam
   )
   n6 = RigVMVariableNode() {variable=curve_struct_var}
   n7 = &RigVMFunction_ControlFlowBranch(ExecuteContext=@n5.ExecutePin, Condition=@n6.Value.deformRig_l_upperarm_fr_alpha_74_#3)
   RigVMDispatchNode(BoneName=root, Name=deformRig_l_upperarm_fr_alpha, ExecuteContext=@n7.True, Value=@n2.OutputParam) {template="DISPATCH_RigDispatch_SetAnimAttribute(in Name,in BoneName,in Value,out Success)"}
   &RigUnit_ParentConstraint(
    AdvancedSettings=(RotationOrderForFilter=XZY)
    Child=(Type=Bone,Name="deform_upperarm_fr_base_r")
    Parents=((Item=(Type=Bone,Name="upperarm_twist_01_r"),Weight=0.3),(Item=(Type=Bone,Name="clavicle_r"),Weight=0.7))
    Weight=1
    ExecutePin=@SequenceExecution_3_1.B
   )
   > n8 = &RigUnit_SphericalPoseReader(
    ActiveRegionScaleFactors=(PositiveWidth=0.5,NegativeWidth=3,PositiveHeight=1.5,NegativeHeight=1.5)
    ActiveRegionSize=0.1
    Debug=(bDrawDebug=false,DebugScale=25,DebugSegments=20,DebugThickness=0.25)
    DriverAxis=(X=-1,Y=0,Z=0)
    DriverItem=(Type=Bone,Name="upperarm_r")
    FalloffRegionScaleFactors=(PositiveWidth=1,NegativeWidth=1,PositiveHeight=1,NegativeHeight=1)
    FalloffSize=0.4
    FlipWidthScaling=true
    OptionalParentItem=(Type=Bone)
    RotationOffset=(X=90,Y=180,Z=180)
    Debug.bDrawDebug=@n1.debug
   )
   > DeformRig_GetMirrorTransform_FNC_2 = RigVMFunctionReferenceNode(transform=@n3.Value.upperarm_fwd_fr_19_#1)
   > &RigUnit_ModifyTransforms(
    ItemToModify=((Item=(Type=Bone,Name="deform_upperarm_fr_r"),Transform=(Rotation=(X=0,Y=0,Z=0,W=1),Translation=(X=0,Y=0,Z=0),Scale3D=(X=1,Y=1,Z=1))))
    ItemToModify.0.Transform=@DeformRig_GetMirrorTransform_FNC_2.mirrorTransform
    Weight=@n8.OutputParam
   )
   > DeformRig_GetMirrorTransform_FNC_1 = RigVMFunctionReferenceNode(transform=@n4.Value.upperarm_fwd_bk_20_#2)
   > n9 = &RigUnit_ModifyTransforms(
    ItemToModify=((Item=(Type=Bone,Name="deform_upperarm_bk_r"),Transform=(Rotation=(X=0,Y=0,Z=0,W=1),Translation=(X=0,Y=0,Z=0),Scale3D=(X=1,Y=1,Z=1))))
    ItemToModify.0.Transform=@DeformRig_GetMirrorTransform_FNC_1.mirrorTransform
    Weight=@n8.OutputParam
   )
   n10 = &RigVMFunction_ControlFlowBranch(ExecuteContext=@n9.ExecutePin, Condition=@n6.Value.deformRig_r_upperarm_fr_alpha_75_#4)
   RigVMDispatchNode(BoneName=root, Name=deformRig_r_upperarm_fr_alpha, ExecuteContext=@n10.True, Value=@n8.OutputParam) {template="DISPATCH_RigDispatch_SetAnimAttribute(in Name,in BoneName,in Value,out Success)"}
   SequenceExecution_3_1_1 = RigVMAggregateNode(ExecuteContext=@SequenceExecution_2.C)
   &RigUnit_ParentConstraint(
    AdvancedSettings=(RotationOrderForFilter=XZY)
    Child=(Type=Bone,Name="deform_upperarm_bk_base_l")
    Parents=((Item=(Type=Bone,Name="upperarm_twist_01_l"),Weight=0.3),(Item=(Type=Bone,Name="clavicle_l"),Weight=0.7))
    Weight=1
    ExecutePin=@SequenceExecution_3_1_1.A
   )
   > n11 = &RigUnit_SphericalPoseReader(
    ActiveRegionScaleFactors=(PositiveWidth=3,NegativeWidth=1,PositiveHeight=1.5,NegativeHeight=1.5)
    ActiveRegionSize=0.1
    Debug=(bDrawDebug=false,DebugScale=25,DebugSegments=20,DebugThickness=0.25)
    DriverAxis=(X=1,Y=0,Z=0)
    DriverItem=(Type=Bone,Name="upperarm_l")
    FalloffRegionScaleFactors=(PositiveWidth=1,NegativeWidth=1,PositiveHeight=1,NegativeHeight=1)
    FalloffSize=0.4
    OptionalParentItem=(Type=Bone)
    RotationOffset=(X=90,Y=0,Z=180)
    Debug.bDrawDebug=@n1.debug
   )
   n12 = RigVMVariableNode() {variable=value_struct_var}
   > n13 = &RigUnit_ModifyTransforms(
    ItemToModify=((Item=(Type=Bone,Name="deform_upperarm_bk_l"),Transform=(Rotation=(X=0,Y=0,Z=0,W=1),Translation=(X=0,Y=0,Z=0),Scale3D=(X=1,Y=1,Z=1))))
    ItemToModify.0.Transform=@n12.Value.upperarm_bk_22_#5
    Weight=@n11.OutputParam
   )
   n14 = RigVMVariableNode() {variable=curve_struct_var}
   n15 = &RigVMFunction_ControlFlowBranch(ExecuteContext=@n13.ExecutePin, Condition=@n14.Value.deformRig_l_upperarm_bk_alpha_76_#6)
   RigVMDispatchNode(BoneName=root, Name=deformRig_l_upperarm_bk_alpha, ExecuteContext=@n15.True, Value=@n11.OutputParam) {template="DISPATCH_RigDispatch_SetAnimAttribute(in Name,in BoneName,in Value,out Success)"}
   &RigUnit_ParentConstraint(
    AdvancedSettings=(RotationOrderForFilter=XZY)
    Child=(Type=Bone,Name="deform_upperarm_bk_base_r")
    Parents=((Item=(Type=Bone,Name="upperarm_twist_01_r"),Weight=0.3),(Item=(Type=Bone,Name="clavicle_r"),Weight=0.7))
    Weight=1
    ExecutePin=@SequenceExecution_3_1_1.B
   )
   > n16 = &RigUnit_SphericalPoseReader(
    ActiveRegionScaleFactors=(PositiveWidth=3,NegativeWidth=1,PositiveHeight=1.5,NegativeHeight=1.5)
    ActiveRegionSize=0.1
    Debug=(bDrawDebug=false,DebugScale=25,DebugSegments=20,DebugThickness=0.25)
    DriverAxis=(X=-1,Y=0,Z=0)
    DriverItem=(Type=Bone,Name="upperarm_r")
    FalloffRegionScaleFactors=(PositiveWidth=1,NegativeWidth=1,PositiveHeight=1,NegativeHeight=1)
    FalloffSize=0.4
    FlipHeightScaling=true
    FlipWidthScaling=false
    OptionalParentItem=(Type=Bone)
    RotationOffset=(X=90,Y=0,Z=0)
    Debug.bDrawDebug=@n1.debug
   )
   > DeformRig_GetMirrorTransform_FNC_3 = RigVMFunctionReferenceNode(transform=@n12.Value.upperarm_bk_22_#5)
   > n17 = &RigUnit_ModifyTransforms(
    ItemToModify=((Item=(Type=Bone,Name="deform_upperarm_bk_r"),Transform=(Rotation=(X=0,Y=0,Z=0,W=1),Translation=(X=0,Y=0,Z=0),Scale3D=(X=1,Y=1,Z=1))))
    ItemToModify.0.Transform=@DeformRig_GetMirrorTransform_FNC_3.mirrorTransform
    Weight=@n16.OutputParam
   )
   n18 = &RigVMFunction_ControlFlowBranch(ExecuteContext=@n17.ExecutePin, Condition=@n14.Value.deformRig_r_upperarm_bk_alpha_77_#7)
   RigVMDispatchNode(BoneName=root, Name=deformRig_r_upperarm_bk_alpha, ExecuteContext=@n18.True, Value=@n16.OutputParam) {template="DISPATCH_RigDispatch_SetAnimAttribute(in Name,in BoneName,in Value,out Success)"}
   SequenceExecution_3 = RigVMAggregateNode(ExecuteContext=@SequenceExecution_2.A)
   &RigUnit_ParentConstraint(
    AdvancedSettings=(RotationOrderForFilter=XZY)
    Child=(Type=Bone,Name="deform_upperarm_up_base_l")
    Parents=((Item=(Type=Bone,Name="upperarm_twist_01_l"),Weight=0.5),(Item=(Type=Bone,Name="clavicle_l"),Weight=0.5))
    Weight=1
    ExecutePin=@SequenceExecution_3.A
   )
   > n19 = &RigUnit_SphericalPoseReader(
    ActiveRegionScaleFactors=(PositiveWidth=2.5,NegativeWidth=3,PositiveHeight=3,NegativeHeight=3)
    ActiveRegionSize=0.1
    Debug=(bDrawDebug=false,DebugScale=25,DebugSegments=20,DebugThickness=0.25)
    DriverAxis=(X=1,Y=0,Z=0)
    DriverItem=(Type=Bone,Name="upperarm_l")
    FalloffRegionScaleFactors=(PositiveWidth=1,NegativeWidth=1,PositiveHeight=1,NegativeHeight=1)
    FalloffSize=0.5
    OptionalParentItem=(Type=Bone)
    RotationOffset=(X=0,Y=0,Z=0)
    Debug.bDrawDebug=@n1.debug
   )
   n20 = RigVMVariableNode() {variable=value_struct_var}
   > n21 = &RigUnit_ModifyTransforms(
    ItemToModify=((Item=(Type=Bone,Name="deform_upperarm_up_l"),Transform=(Rotation=(X=0,Y=0,Z=0,W=1),Translation=(X=0,Y=0,Z=0),Scale3D=(X=1,Y=1,Z=1))))
    ItemToModify.0.Transform=@n20.Value.upperarm_up_18_#8
    Weight=@n19.OutputParam
   )
   n22 = RigVMVariableNode() {variable=curve_struct_var}
   RigVMFunction_ControlFlowBranch = &RigVMFunction_ControlFlowBranch(ExecuteContext=@n21.ExecutePin, Condition=@n22.Value.deformRig_l_upperarm_up_alpha_72_#9)
   RigVMDispatchNode(BoneName=root, Name=deformRig_l_upperarm_up_alpha, ExecuteContext=@RigVMFunction_ControlFlowBranch.True, Value=@n19.OutputParam) {template="DISPATCH_RigDispatch_SetAnimAttribute(in Name,in BoneName,in Value,out Success)"}
   &RigUnit_ParentConstraint(
    AdvancedSettings=(RotationOrderForFilter=XZY)
    Child=(Type=Bone,Name="deform_upperarm_up_base_r")
    Parents=((Item=(Type=Bone,Name="upperarm_twist_01_r"),Weight=0.5),(Item=(Type=Bone,Name="clavicle_r"),Weight=0.5))
    Weight=1
    ExecutePin=@SequenceExecution_3.B
   )
   > n23 = &RigUnit_SphericalPoseReader(
    ActiveRegionScaleFactors=(PositiveWidth=2.5,NegativeWidth=3,PositiveHeight=3,NegativeHeight=3)
    ActiveRegionSize=0.1
    Debug=(bDrawDebug=false,DebugScale=25,DebugSegments=20,DebugThickness=0.25)
    DriverAxis=(X=-1,Y=0,Z=0)
    DriverItem=(Type=Bone,Name="upperarm_r")
    FalloffRegionScaleFactors=(PositiveWidth=1,NegativeWidth=1,PositiveHeight=1,NegativeHeight=1)
    FalloffSize=0.5
    FlipHeightScaling=true
    OptionalParentItem=(Type=Bone)
    RotationOffset=(X=180,Y=0,Z=180)
    Debug.bDrawDebug=@n1.debug
   )
   > DeformRig_GetMirrorTransform_FNC = RigVMFunctionReferenceNode(transform=@n20.Value.upperarm_up_18_#8)
   > n24 = &RigUnit_ModifyTransforms(
    ItemToModify=((Item=(Type=Bone,Name="deform_upperarm_up_r"),Transform=(Rotation=(X=0,Y=0,Z=0,W=1),Translation=(X=0,Y=0,Z=0),Scale3D=(X=1,Y=1,Z=1))))
    ItemToModify.0.Transform=@DeformRig_GetMirrorTransform_FNC.mirrorTransform
    Weight=@n23.OutputParam
   )
   n25 = &RigVMFunction_ControlFlowBranch(ExecuteContext=@n24.ExecutePin, Condition=@n22.Value.deformRig_r_upperarm_up_alpha_73_#10)
   RigVMDispatchNode(BoneName=root, Name=deformRig_r_upperarm_up_alpha, ExecuteContext=@n25.True, Value=@n23.OutputParam) {template="DISPATCH_RigDispatch_SetAnimAttribute(in Name,in BoneName,in Value,out Success)"}
  ]}
 ContainedGraph {nodes [
   n1 = RigVMFunctionEntryNode()
   SequenceExecution_2 = &RigVMFunction_Sequence(ExecuteContext=@n1.ExecuteContext)
   SequenceExecution_2_1 = &RigVMFunction_Sequence(ExecuteContext=@SequenceExecution_2.B)
   SequenceExecution_2_2 = &RigVMFunction_Sequence(ExecuteContext=@SequenceExecution_2_1.B)
   RigVMFunctionReturnNode(A=@SequenceExecution_2.A, B=@SequenceExecution_2_1.A, C=@SequenceExecution_2_2.A, D=@SequenceExecution_2_2.B)
  ]}
 ContainedGraph {nodes [
   n1 = RigVMFunctionEntryNode()
   SequenceExecution_3_1 = &RigVMFunction_Sequence(ExecuteContext=@n1.ExecuteContext)
   SequenceExecution_3_1_1 = &RigVMFunction_Sequence(ExecuteContext=@SequenceExecution_3_1.B)
   SequenceExecution_3_1_2 = &RigVMFunction_Sequence(ExecuteContext=@SequenceExecution_3_1_1.B)
   RigVMFunctionReturnNode(A=@SequenceExecution_3_1.A, B=@SequenceExecution_3_1_1.A, C=@SequenceExecution_3_1_2.A, D=@SequenceExecution_3_1_2.B)
  ]}
 ContainedGraph {nodes [
   n1 = RigVMFunctionEntryNode()
   SequenceExecution_3 = &RigVMFunction_Sequence(ExecuteContext=@n1.ExecuteContext)
   SequenceExecution_3_1 = &RigVMFunction_Sequence(ExecuteContext=@SequenceExecution_3.B)
   SequenceExecution_3_2 = &RigVMFunction_Sequence(ExecuteContext=@SequenceExecution_3_1.B)
   RigVMFunctionReturnNode(A=@SequenceExecution_3.A, B=@SequenceExecution_3_1.A, C=@SequenceExecution_3_2.A, D=@SequenceExecution_3_2.B)
  ]}
 ContainedGraph {nodes [
   n1 = RigVMFunctionEntryNode()
   SequenceExecution_3_1_1 = &RigVMFunction_Sequence(ExecuteContext=@n1.ExecuteContext)
   SequenceExecution_3_1_1_1 = &RigVMFunction_Sequence(ExecuteContext=@SequenceExecution_3_1_1.B)
   SequenceExecution_3_1_1_2 = &RigVMFunction_Sequence(ExecuteContext=@SequenceExecution_3_1_1_1.B)
   RigVMFunctionReturnNode(A=@SequenceExecution_3_1_1.A, B=@SequenceExecution_3_1_1_1.A, C=@SequenceExecution_3_1_1_2.A, D=@SequenceExecution_3_1_1_2.B)
  ]}
 ContainedGraph {nodes [
   n1 = RigVMFunctionEntryNode()
   > RigVMVariableNode(Value=@n1.value_struct) {variable=value_struct_var}
   > RigVMVariableNode(Value=@n1.curve_struct) {variable=curve_struct_var}
   > SequenceExecution = RigVMAggregateNode()
   RigVMFunctionReturnNode(ExecuteContext=@SequenceExecution.D)
   SequenceExecution_1 = RigVMAggregateNode(ExecuteContext=@SequenceExecution.A)
   &RigUnit_ParentConstraint(
    AdvancedSettings=(RotationOrderForFilter=XZY)
    Child=(Type=Bone,Name="deform_wrist_up_base_l")
    Parents=((Item=(Type=Bone,Name="hand_l"),Weight=0.4),(Item=(Type=Bone,Name="lowerarm_twist_01_l"),Weight=0.6))
    Weight=1
    ExecutePin=@SequenceExecution_1.A
   )
   > n2 = &RigUnit_SphericalPoseReader(
    ActiveRegionScaleFactors=(PositiveWidth=0.5,NegativeWidth=3.5,PositiveHeight=3,NegativeHeight=3)
    ActiveRegionSize=0.1
    Debug=(bDrawDebug=false,DebugScale=25,DebugSegments=20,DebugThickness=0.25)
    DriverAxis=(X=1,Y=0,Z=0)
    DriverItem=(Type=Bone,Name="hand_l")
    FalloffRegionScaleFactors=(PositiveWidth=1,NegativeWidth=1,PositiveHeight=1,NegativeHeight=1)
    FalloffSize=0.45
    OptionalParentItem=(Type=Bone,Name="lowerarm_twist_01_l")
    RotationOffset=(X=90,Y=0,Z=0)
    Debug.bDrawDebug=@n1.debug
   )
   n3 = RigVMVariableNode() {variable=value_struct_var}
   > n4 = &RigUnit_ModifyTransforms(
    ItemToModify=((Item=(Type=Bone,Name="deform_wrist_up_l"),Transform=(Rotation=(X=0,Y=0,Z=0,W=1),Translation=(X=0,Y=0,Z=0),Scale3D=(X=1,Y=1,Z=1))))
    ItemToModify.0.Transform=@n3.Value.wrist_up_31_#11
    Weight=@n2.OutputParam
   )
   n5 = RigVMVariableNode() {variable=curve_struct_var}
   RigVMFunction_ControlFlowBranch = &RigVMFunction_ControlFlowBranch(ExecuteContext=@n4.ExecutePin, Condition=@n5.Value.deformRig_l_wrist_up_alpha_78_#12)
   RigVMDispatchNode(BoneName=root, Name=deformRig_l_wrist_up_alpha, ExecuteContext=@RigVMFunction_ControlFlowBranch.True, Value=@n2.OutputParam) {template="DISPATCH_RigDispatch_SetAnimAttribute(in Name,in BoneName,in Value,out Success)"}
   &RigUnit_ParentConstraint(
    AdvancedSettings=(RotationOrderForFilter=XZY)
    Child=(Type=Bone,Name="deform_wrist_up_base_r")
    Parents=((Item=(Type=Bone,Name="hand_r"),Weight=0.4),(Item=(Type=Bone,Name="lowerarm_twist_01_r"),Weight=0.6))
    Weight=1
    ExecutePin=@SequenceExecution_1.B
   )
   > n6 = &RigUnit_SphericalPoseReader(
    ActiveRegionScaleFactors=(PositiveWidth=0.5,NegativeWidth=3.5,PositiveHeight=3,NegativeHeight=3)
    ActiveRegionSize=0.1
    Debug=(bDrawDebug=false,DebugScale=25,DebugSegments=20,DebugThickness=0.25)
    DriverAxis=(X=-1,Y=0,Z=0)
    DriverItem=(Type=Bone,Name="hand_r")
    FalloffRegionScaleFactors=(PositiveWidth=1,NegativeWidth=1,PositiveHeight=1,NegativeHeight=1)
    FalloffSize=0.45
    FlipHeightScaling=true
    OptionalParentItem=(Type=Bone,Name="lowerarm_twist_01_r")
    RotationOffset=(X=90,Y=0,Z=180)
    Debug.bDrawDebug=@n1.debug
   )
   > DeformRig_GetMirrorTransform_FNC = RigVMFunctionReferenceNode(transform=@n3.Value.wrist_up_31_#11)
   > n7 = &RigUnit_ModifyTransforms(
    ItemToModify=((Item=(Type=Bone,Name="deform_wrist_up_r"),Transform=(Rotation=(X=0,Y=0,Z=0,W=1),Translation=(X=0,Y=0,Z=0),Scale3D=(X=1,Y=1,Z=1))))
    ItemToModify.0.Transform=@DeformRig_GetMirrorTransform_FNC.mirrorTransform
    Weight=@n6.OutputParam
   )
   n8 = &RigVMFunction_ControlFlowBranch(ExecuteContext=@n7.ExecutePin, Condition=@n5.Value.deformRig_r_wrist_up_alpha_79_#13)
   RigVMDispatchNode(BoneName=root, Name=deformRig_r_wrist_up_alpha, ExecuteContext=@n8.True, Value=@n6.OutputParam) {template="DISPATCH_RigDispatch_SetAnimAttribute(in Name,in BoneName,in Value,out Success)"}
   SequenceExecution_1_1 = RigVMAggregateNode(ExecuteContext=@SequenceExecution.B)
   &RigUnit_ParentConstraint(
    AdvancedSettings=(RotationOrderForFilter=XZY)
    Child=(Type=Bone,Name="deform_wrist_dn_base_l")
    Parents=((Item=(Type=Bone,Name="hand_l"),Weight=0.4),(Item=(Type=Bone,Name="lowerarm_twist_01_l"),Weight=0.6))
    Weight=1
    ExecutePin=@SequenceExecution_1_1.A
   )
   > n9 = &RigUnit_SphericalPoseReader(
    ActiveRegionScaleFactors=(PositiveWidth=1,NegativeWidth=4,PositiveHeight=3,NegativeHeight=3)
    ActiveRegionSize=0.1
    Debug=(bDrawDebug=false,DebugScale=25,DebugSegments=20,DebugThickness=0.25)
    DriverAxis=(X=1,Y=0,Z=0)
    DriverItem=(Type=Bone,Name="hand_l")
    FalloffRegionScaleFactors=(PositiveWidth=1,NegativeWidth=1,PositiveHeight=1,NegativeHeight=1)
    FalloffSize=0.4
    OptionalParentItem=(Type=Bone,Name="lowerarm_twist_01_l")
    RotationOffset=(X=-90,Y=0,Z=0)
    Debug.bDrawDebug=@n1.debug
   )
   n10 = RigVMVariableNode() {variable=value_struct_var}
   > n11 = &RigUnit_ModifyTransforms(
    ItemToModify=((Item=(Type=Bone,Name="deform_wrist_dn_l"),Transform=(Rotation=(X=0,Y=0,Z=0,W=1),Translation=(X=0,Y=0,Z=0),Scale3D=(X=1,Y=1,Z=1))))
    ItemToModify.0.Transform=@n10.Value.wrist_dn_32_#14
    Weight=@n9.OutputParam
   )
   n12 = RigVMVariableNode() {variable=curve_struct_var}
   n13 = &RigVMFunction_ControlFlowBranch(ExecuteContext=@n11.ExecutePin, Condition=@n12.Value.deformRig_l_wrist_dn_alpha_80_#15)
   RigVMDispatchNode(BoneName=root, Name=deformRig_l_wrist_dn_alpha, ExecuteContext=@n13.True, Value=@n9.OutputParam) {template="DISPATCH_RigDispatch_SetAnimAttribute(in Name,in BoneName,in Value,out Success)"}
   &RigUnit_ParentConstraint(
    AdvancedSettings=(RotationOrderForFilter=XZY)
    Child=(Type=Bone,Name="deform_wrist_dn_base_r")
    Parents=((Item=(Type=Bone,Name="hand_r"),Weight=0.4),(Item=(Type=Bone,Name="lowerarm_twist_01_r"),Weight=0.6))
    Weight=1
    ExecutePin=@SequenceExecution_1_1.B
   )
   > n14 = &RigUnit_SphericalPoseReader(
    ActiveRegionScaleFactors=(PositiveWidth=1,NegativeWidth=4,PositiveHeight=3,NegativeHeight=3)
    ActiveRegionSize=0.1
    Debug=(bDrawDebug=false,DebugScale=25,DebugSegments=20,DebugThickness=0.25)
    DriverAxis=(X=-1,Y=0,Z=0)
    DriverItem=(Type=Bone,Name="hand_r")
    FalloffRegionScaleFactors=(PositiveWidth=1,NegativeWidth=1,PositiveHeight=1,NegativeHeight=1)
    FalloffSize=0.4
    FlipHeightScaling=true
    OptionalParentItem=(Type=Bone,Name="lowerarm_twist_01_r")
    RotationOffset=(X=-90,Y=0,Z=180)
    Debug.bDrawDebug=@n1.debug
   )
   > DeformRig_GetMirrorTransform_FNC_1 = RigVMFunctionReferenceNode(transform=@n10.Value.wrist_dn_32_#14)
   > n15 = &RigUnit_ModifyTransforms(
    ItemToModify=((Item=(Type=Bone,Name="deform_wrist_dn_r"),Transform=(Rotation=(X=0,Y=0,Z=0,W=1),Translation=(X=0,Y=0,Z=0),Scale3D=(X=1,Y=1,Z=1))))
    ItemToModify.0.Transform=@DeformRig_GetMirrorTransform_FNC_1.mirrorTransform
    Weight=@n14.OutputParam
   )
   n16 = &RigVMFunction_ControlFlowBranch(ExecuteContext=@n15.ExecutePin, Condition=@n12.Value.deformRig_r_wrist_dn_alpha_81_#16)
   RigVMDispatchNode(BoneName=root, Name=deformRig_r_wrist_dn_alpha, ExecuteContext=@n16.True, Value=@n14.OutputParam) {template="DISPATCH_RigDispatch_SetAnimAttribute(in Name,in BoneName,in Value,out Success)"}
  ]}
 ContainedGraph {nodes [
   n1 = RigVMFunctionEntryNode()
   SequenceExecution = &RigVMFunction_Sequence(ExecuteContext=@n1.ExecuteContext)
   SequenceExecution_1 = &RigVMFunction_Sequence(ExecuteContext=@SequenceExecution.B)
   SequenceExecution_2 = &RigVMFunction_Sequence(ExecuteContext=@SequenceExecution_1.B)
   RigVMFunctionReturnNode(A=@SequenceExecution.A, B=@SequenceExecution_1.A, C=@SequenceExecution_2.A, D=@SequenceExecution_2.B)
  ]}
 ContainedGraph {nodes [
   n1 = RigVMFunctionEntryNode()
   SequenceExecution_1 = &RigVMFunction_Sequence(ExecuteContext=@n1.ExecuteContext)
   SequenceExecution_1_1 = &RigVMFunction_Sequence(ExecuteContext=@SequenceExecution_1.B)
   SequenceExecution_1_2 = &RigVMFunction_Sequence(ExecuteContext=@SequenceExecution_1_1.B)
   RigVMFunctionReturnNode(A=@SequenceExecution_1.A, B=@SequenceExecution_1_1.A, C=@SequenceExecution_1_2.A, D=@SequenceExecution_1_2.B)
  ]}
 ContainedGraph {nodes [
   n1 = RigVMFunctionEntryNode()
   SequenceExecution_1_1 = &RigVMFunction_Sequence(ExecuteContext=@n1.ExecuteContext)
   SequenceExecution_1_1_1 = &RigVMFunction_Sequence(ExecuteContext=@SequenceExecution_1_1.B)
   SequenceExecution_1_1_2 = &RigVMFunction_Sequence(ExecuteContext=@SequenceExecution_1_1_1.B)
   RigVMFunctionReturnNode(A=@SequenceExecution_1_1.A, B=@SequenceExecution_1_1_1.A, C=@SequenceExecution_1_1_2.A, D=@SequenceExecution_1_1_2.B)
  ]}
 ContainedGraph {nodes [
   n1 = RigVMFunctionEntryNode()
   > RigVMVariableNode(Value=@n1.value_struct) {variable=value_struct_var}
   > RigVMVariableNode(Value=@n1.curve_struct) {variable=curve_struct_var}
   > SequenceExecution = RigVMAggregateNode()
   n2 = &RigUnit_SphericalPoseReader(
    ActiveRegionScaleFactors=(PositiveWidth=3.5,NegativeWidth=0.5,PositiveHeight=3,NegativeHeight=3)
    ActiveRegionSize=0.1
    Debug=(bDrawDebug=false,DebugScale=25,DebugSegments=20,DebugThickness=0.25)
    DriverAxis=(X=-1,Y=0,Z=0)
    DriverItem=(Type=Bone,Name="thigh_l")
    FalloffRegionScaleFactors=(PositiveWidth=1,NegativeWidth=1,PositiveHeight=1,NegativeHeight=1)
    FalloffSize=0.45
    OptionalParentItem=(Type=Bone)
    RotationOffset=(X=90,Y=0,Z=0)
    ExecutePin=@SequenceExecution.A
    Debug.bDrawDebug=@n1.debug
   )
   > RigVMVariableNode(Value=@n2.OutputParam) {variable=l_thigh_up_alpha}
   > n3 = &RigUnit_SphericalPoseReader(
    ActiveRegionScaleFactors=(PositiveWidth=3.5,NegativeWidth=0.5,PositiveHeight=3,NegativeHeight=3)
    ActiveRegionSize=0.1
    Debug=(bDrawDebug=false,DebugScale=25,DebugSegments=20,DebugThickness=0.25)
    DriverAxis=(X=1,Y=0,Z=0)
    DriverItem=(Type=Bone,Name="thigh_r")
    FalloffRegionScaleFactors=(PositiveWidth=1,NegativeWidth=1,PositiveHeight=1,NegativeHeight=1)
    FalloffSize=0.45
    FlipHeightScaling=false
    FlipWidthScaling=true
    OptionalParentItem=(Type=Bone)
    RotationOffset=(X=-90,Y=0,Z=0)
    Debug.bDrawDebug=@n1.debug
   )
   > RigVMVariableNode(Value=@n3.OutputParam) {variable=r_thigh_up_alpha}
   RigVMFunctionReturnNode(ExecuteContext=@SequenceExecution.E)
   n4 = &RigVMFunction_ControlFlowBranch(ExecuteContext=@SequenceExecution.C, Condition=@n1.glute)
   SequenceExecution_1_1 = RigVMAggregateNode(ExecuteContext=@n4.True)
   &RigUnit_ParentConstraint(
    AdvancedSettings=(RotationOrderForFilter=XZY)
    Child=(Type=Bone,Name="deform_glute_base_l")
    Parents=((Item=(Type=Bone,Name="thigh_l"),Weight=0.25),(Item=(Type=Bone,Name="pelvis"),Weight=0.75))
    Weight=1
    ExecutePin=@SequenceExecution_1_1.A
   )
   n5 = RigVMVariableNode() {variable=value_struct_var}
   > &RigUnit_ModifyTransforms(
    ItemToModify=((Item=(Type=Bone,Name="deform_glute_l"),Transform=(Rotation=(X=0,Y=0,Z=0,W=1),Translation=(X=0,Y=0,Z=0),Scale3D=(X=1,Y=1,Z=1))))
    ItemToModify.0.Transform=@n5.Value.glute_36_#17
    Weight=RigVMVariableNode() {variable=l_thigh_up_alpha}.Value
   )
   &RigUnit_ParentConstraint(
    AdvancedSettings=(RotationOrderForFilter=XZY)
    Child=(Type=Bone,Name="deform_glute_base_r")
    Parents=((Item=(Type=Bone,Name="thigh_r"),Weight=0.25),(Item=(Type=Bone,Name="pelvis"),Weight=0.75))
    Weight=1
    ExecutePin=@SequenceExecution_1_1.B
   )
   > DeformRig_GetMirrorTransform_FNC_1 = RigVMFunctionReferenceNode(transform=@n5.Value.glute_36_#17)
   > &RigUnit_ModifyTransforms(
    ItemToModify=((Item=(Type=Bone,Name="deform_glute_r"),Transform=(Rotation=(X=0,Y=0,Z=0,W=1),Translation=(X=0,Y=0,Z=0),Scale3D=(X=1,Y=1,Z=1))))
    ItemToModify.0.Transform=@DeformRig_GetMirrorTransform_FNC_1.mirrorTransform
    Weight=RigVMVariableNode() {variable=r_thigh_up_alpha}.Value
   )
   RigVMFunction_ControlFlowBranch = &RigVMFunction_ControlFlowBranch(ExecuteContext=@SequenceExecution.B, Condition=@n1.groin)
   SequenceExecution_1 = RigVMAggregateNode(ExecuteContext=@RigVMFunction_ControlFlowBranch.True)
   &RigUnit_ParentConstraint(
    AdvancedSettings=(RotationOrderForFilter=XZY)
    Child=(Type=Bone,Name="deform_groin_base_l")
    Parents=((Item=(Type=Bone,Name="thigh_l"),Weight=0.7),(Item=(Type=Bone,Name="pelvis"),Weight=0.3))
    Weight=1
    ExecutePin=@SequenceExecution_1.A
   )
   n6 = RigVMVariableNode() {variable=value_struct_var}
   > &RigUnit_ModifyTransforms(
    ItemToModify=((Item=(Type=Bone,Name="deform_groin_l"),Transform=(Rotation=(X=0,Y=0,Z=0,W=1),Translation=(X=0,Y=0,Z=0),Scale3D=(X=1,Y=1,Z=1))))
    ItemToModify.0.Transform=@n6.Value.groin_35_#18
    Weight=RigVMVariableNode() {variable=l_thigh_up_alpha}.Value
   )
   &RigUnit_ParentConstraint(
    AdvancedSettings=(RotationOrderForFilter=XZY)
    Child=(Type=Bone,Name="deform_groin_base_r")
    Parents=((Item=(Type=Bone,Name="thigh_r"),Weight=0.7),(Item=(Type=Bone,Name="pelvis"),Weight=0.3))
    Weight=1
    ExecutePin=@SequenceExecution_1.B
   )
   > DeformRig_GetMirrorTransform_FNC = RigVMFunctionReferenceNode(transform=@n6.Value.groin_35_#18)
   > &RigUnit_ModifyTransforms(
    ItemToModify=((Item=(Type=Bone,Name="deform_groin_r"),Transform=(Rotation=(X=0,Y=0,Z=0,W=1),Translation=(X=0,Y=0,Z=0),Scale3D=(X=1,Y=1,Z=1))))
    ItemToModify.0.Transform=@DeformRig_GetMirrorTransform_FNC.mirrorTransform
    Weight=RigVMVariableNode() {variable=r_thigh_up_alpha}.Value
   )
   RigVMFunction_Sequence = &RigVMFunction_Sequence(ExecuteContext=@SequenceExecution.D)
   n7 = RigVMVariableNode() {variable=curve_struct_var}
   n8 = &RigVMFunction_ControlFlowBranch(ExecuteContext=@RigVMFunction_Sequence.A, Condition=@n7.Value.deformRig_l_thigh_up_alpha_88_#19)
   RigVMDispatchNode(BoneName=root, Name=deformRig_l_thigh_up_alpha, ExecuteContext=@n8.True, Value=RigVMVariableNode() {variable=l_thigh_up_alpha}.Value) {template="DISPATCH_RigDispatch_SetAnimAttribute(in Name,in BoneName,in Value,out Success)"}
   n9 = &RigVMFunction_ControlFlowBranch(ExecuteContext=@RigVMFunction_Sequence.B, Condition=@n7.Value.deformRig_r_thigh_up_alpha_89_#20)
   RigVMDispatchNode(BoneName=root, Name=deformRig_r_thigh_up_alpha, ExecuteContext=@n9.True, Value=RigVMVariableNode() {variable=r_thigh_up_alpha}.Value) {template="DISPATCH_RigDispatch_SetAnimAttribute(in Name,in BoneName,in Value,out Success)"}
  ]}
 ContainedGraph {nodes [
   n1 = RigVMFunctionEntryNode()
   SequenceExecution = &RigVMFunction_Sequence(ExecuteContext=@n1.ExecuteContext)
   SequenceExecution_1 = &RigVMFunction_Sequence(ExecuteContext=@SequenceExecution.B)
   SequenceExecution_2 = &RigVMFunction_Sequence(ExecuteContext=@SequenceExecution_1.B)
   SequenceExecution_3 = &RigVMFunction_Sequence(ExecuteContext=@SequenceExecution_2.B)
   RigVMFunctionReturnNode(A=@SequenceExecution.A, B=@SequenceExecution_1.A, C=@SequenceExecution_2.A, D=@SequenceExecution_3.A, E=@SequenceExecution_3.B)
  ]}
 ContainedGraph {nodes [
   n1 = RigVMFunctionEntryNode()
   SequenceExecution_1 = &RigVMFunction_Sequence(ExecuteContext=@n1.ExecuteContext)
   SequenceExecution_1_1 = &RigVMFunction_Sequence(ExecuteContext=@SequenceExecution_1.B)
   SequenceExecution_1_2 = &RigVMFunction_Sequence(ExecuteContext=@SequenceExecution_1_1.B)
   RigVMFunctionReturnNode(A=@SequenceExecution_1.A, B=@SequenceExecution_1_1.A, C=@SequenceExecution_1_2.A, D=@SequenceExecution_1_2.B)
  ]}
 ContainedGraph {nodes [
   n1 = RigVMFunctionEntryNode()
   SequenceExecution_1_1 = &RigVMFunction_Sequence(ExecuteContext=@n1.ExecuteContext)
   SequenceExecution_1_1_1 = &RigVMFunction_Sequence(ExecuteContext=@SequenceExecution_1_1.B)
   SequenceExecution_1_1_2 = &RigVMFunction_Sequence(ExecuteContext=@SequenceExecution_1_1_1.B)
   RigVMFunctionReturnNode(A=@SequenceExecution_1_1.A, B=@SequenceExecution_1_1_1.A, C=@SequenceExecution_1_1_2.A, D=@SequenceExecution_1_1_2.B)
  ]}
 ContainedGraph {nodes [
   n1 = RigVMFunctionEntryNode()
   > RigVMVariableNode(Value=@n1.value_struct) {variable=value_struct_var}
   > RigVMVariableNode(Value=@n1.curve_struct) {variable=curve_struct_var}
   > SequenceExecution_2 = RigVMAggregateNode()
   n2 = &RigVMFunction_Sequence(ExecuteContext=@SequenceExecution_2.A)
   DeformRig_GetDeltaRefTransform_FNC = RigVMFunctionReferenceNode(Item=(Type=Bone,Name="calf_l"), ExecuteContext=@n2.A)
   > RigVMVariableNode(Value=&RigVMFunction_MathFloatRemap(SourceMaximum=100, SourceMinimum=10, bClamp=true, Value=@DeformRig_GetDeltaRefTransform_FNC.OutTransform.Rotation.Yaw).Result) {variable=l_knee_alpha}
   DeformRig_GetDeltaRefTransform_FNC_1 = RigVMFunctionReferenceNode(Item=(Type=Bone,Name="calf_r"), ExecuteContext=@n2.B)
   > RigVMVariableNode(Value=&RigVMFunction_MathFloatRemap(SourceMaximum=100, SourceMinimum=10, bClamp=true, Value=@DeformRig_GetDeltaRefTransform_FNC_1.OutTransform.Rotation.Yaw).Result) {variable=r_knee_alpha}
   SequenceExecution = RigVMAggregateNode(ExecuteContext=@SequenceExecution_2.B)
   SequenceExecution_1 = RigVMAggregateNode(ExecuteContext=@SequenceExecution.A)
   &RigUnit_ParentConstraint(
    AdvancedSettings=(RotationOrderForFilter=XZY)
    Child=(Type=Bone,Name="deform_knee_in_base_l")
    Parents=((Item=(Type=Bone,Name="calf_l"),Weight=0.5),(Item=(Type=Bone,Name="thigh_l"),Weight=0.5))
    Weight=1
    ExecutePin=@SequenceExecution_1.A
   )
   n3 = RigVMVariableNode() {variable=value_struct_var}
   > &RigUnit_ModifyTransforms(
    ItemToModify=((Item=(Type=Bone,Name="deform_knee_in_l"),Transform=(Rotation=(X=0,Y=0,Z=0,W=1),Translation=(X=0,Y=0,Z=0),Scale3D=(X=1,Y=1,Z=1))))
    ItemToModify.0.Transform=@n3.Value.knee_in_41_#21
    Weight=RigVMVariableNode() {variable=l_knee_alpha}.Value
   )
   &RigUnit_ParentConstraint(
    AdvancedSettings=(RotationOrderForFilter=XZY)
    Child=(Type=Bone,Name="deform_knee_in_base_r")
    Parents=((Item=(Type=Bone,Name="calf_r"),Weight=0.5),(Item=(Type=Bone,Name="thigh_r"),Weight=0.5))
    Weight=1
    ExecutePin=@SequenceExecution_1.B
   )
   > DeformRig_GetMirrorTransform_FNC = RigVMFunctionReferenceNode(transform=@n3.Value.knee_in_41_#21)
   > &RigUnit_ModifyTransforms(
    ItemToModify=((Item=(Type=Bone,Name="deform_knee_in_r"),Transform=(Rotation=(X=0,Y=0,Z=0,W=1),Translation=(X=0,Y=0,Z=0),Scale3D=(X=1,Y=1,Z=1))))
    ItemToModify.0.Transform=@DeformRig_GetMirrorTransform_FNC.mirrorTransform
    Weight=RigVMVariableNode() {variable=r_knee_alpha}.Value
   )
   SequenceExecution_1_1 = RigVMAggregateNode(ExecuteContext=@SequenceExecution.B)
   &RigUnit_ParentConstraint(
    AdvancedSettings=(RotationOrderForFilter=XZY)
    Child=(Type=Bone,Name="deform_knee_out_base_l")
    Parents=((Item=(Type=Bone,Name="calf_l"),Weight=0.1),(Item=(Type=Bone,Name="thigh_l"),Weight=0.9))
    Weight=1
    ExecutePin=@SequenceExecution_1_1.A
   )
   n4 = RigVMVariableNode() {variable=value_struct_var}
   > &RigUnit_ModifyTransforms(
    ItemToModify=((Item=(Type=Bone,Name="deform_knee_out_l"),Transform=(Rotation=(X=0,Y=0,Z=0,W=1),Translation=(X=0,Y=0,Z=0),Scale3D=(X=1,Y=1,Z=1))))
    ItemToModify.0.Transform=@n4.Value.knee_out_42_#22
    Weight=RigVMVariableNode() {variable=l_knee_alpha}.Value
   )
   &RigUnit_ParentConstraint(
    AdvancedSettings=(RotationOrderForFilter=XZY)
    Child=(Type=Bone,Name="deform_knee_out_base_r")
    Parents=((Item=(Type=Bone,Name="calf_r"),Weight=0.1),(Item=(Type=Bone,Name="thigh_r"),Weight=0.9))
    Weight=1
    ExecutePin=@SequenceExecution_1_1.B
   )
   > DeformRig_GetMirrorTransform_FNC_1 = RigVMFunctionReferenceNode(transform=@n4.Value.knee_out_42_#22)
   > &RigUnit_ModifyTransforms(
    ItemToModify=((Item=(Type=Bone,Name="deform_knee_out_r"),Transform=(Rotation=(X=0,Y=0,Z=0,W=1),Translation=(X=0,Y=0,Z=0),Scale3D=(X=1,Y=1,Z=1))))
    ItemToModify.0.Transform=@DeformRig_GetMirrorTransform_FNC_1.mirrorTransform
    Weight=RigVMVariableNode() {variable=r_knee_alpha}.Value
   )
   RigVMFunction_Sequence = &RigVMFunction_Sequence(ExecuteContext=@SequenceExecution_2.C)
   n5 = RigVMVariableNode() {variable=curve_struct_var}
   RigVMFunction_ControlFlowBranch = &RigVMFunction_ControlFlowBranch(ExecuteContext=@RigVMFunction_Sequence.A, Condition=@n5.Value.deformRig_l_knee_alpha_82_#23)
   RigVMDispatchNode(
    BoneName=root
    Name=deformRig_l_knee_alpha
    ExecuteContext=@RigVMFunction_ControlFlowBranch.True
    Value=RigVMVariableNode() {variable=l_knee_alpha}.Value
   ) {template="DISPATCH_RigDispatch_SetAnimAttribute(in Name,in BoneName,in Value,out Success)"}
   n6 = &RigVMFunction_ControlFlowBranch(ExecuteContext=@RigVMFunction_Sequence.B, Condition=@n5.Value.deformRig_r_knee_alpha_83_#24)
   RigVMDispatchNode(BoneName=root, Name=deformRig_r_knee_alpha, ExecuteContext=@n6.True, Value=RigVMVariableNode() {variable=r_knee_alpha}.Value) {template="DISPATCH_RigDispatch_SetAnimAttribute(in Name,in BoneName,in Value,out Success)"}
   RigVMFunctionReturnNode(ExecuteContext=@SequenceExecution_2.D)
  ]}
 ContainedGraph {nodes [
   n1 = RigVMFunctionEntryNode()
   SequenceExecution = &RigVMFunction_Sequence(ExecuteContext=@n1.ExecuteContext)
   SequenceExecution_1 = &RigVMFunction_Sequence(ExecuteContext=@SequenceExecution.B)
   SequenceExecution_2 = &RigVMFunction_Sequence(ExecuteContext=@SequenceExecution_1.B)
   RigVMFunctionReturnNode(A=@SequenceExecution.A, B=@SequenceExecution_1.A, C=@SequenceExecution_2.A, D=@SequenceExecution_2.B)
  ]}
 ContainedGraph {nodes [
   n1 = RigVMFunctionEntryNode()
   SequenceExecution_2 = &RigVMFunction_Sequence(ExecuteContext=@n1.ExecuteContext)
   SequenceExecution_2_1 = &RigVMFunction_Sequence(ExecuteContext=@SequenceExecution_2.B)
   SequenceExecution_2_2 = &RigVMFunction_Sequence(ExecuteContext=@SequenceExecution_2_1.B)
   SequenceExecution_2_3 = &RigVMFunction_Sequence(ExecuteContext=@SequenceExecution_2_2.B)
   RigVMFunctionReturnNode(A=@SequenceExecution_2.A, B=@SequenceExecution_2_1.A, C=@SequenceExecution_2_2.A, D=@SequenceExecution_2_3.A, E=@SequenceExecution_2_3.B)
  ]}
 ContainedGraph {nodes [
   n1 = RigVMFunctionEntryNode()
   SequenceExecution_1_1 = &RigVMFunction_Sequence(ExecuteContext=@n1.ExecuteContext)
   SequenceExecution_1_1_1 = &RigVMFunction_Sequence(ExecuteContext=@SequenceExecution_1_1.B)
   SequenceExecution_1_1_2 = &RigVMFunction_Sequence(ExecuteContext=@SequenceExecution_1_1_1.B)
   RigVMFunctionReturnNode(A=@SequenceExecution_1_1.A, B=@SequenceExecution_1_1_1.A, C=@SequenceExecution_1_1_2.A, D=@SequenceExecution_1_1_2.B)
  ]}
 ContainedGraph {nodes [
   n1 = RigVMFunctionEntryNode()
   SequenceExecution_1 = &RigVMFunction_Sequence(ExecuteContext=@n1.ExecuteContext)
   SequenceExecution_1_1 = &RigVMFunction_Sequence(ExecuteContext=@SequenceExecution_1.B)
   SequenceExecution_1_2 = &RigVMFunction_Sequence(ExecuteContext=@SequenceExecution_1_1.B)
   RigVMFunctionReturnNode(A=@SequenceExecution_1.A, B=@SequenceExecution_1_1.A, C=@SequenceExecution_1_2.A, D=@SequenceExecution_1_2.B)
  ]}
 ContainedGraph {nodes [
   n1 = RigVMFunctionEntryNode()
   > RigVMVariableNode(Value=@n1.value_struct) {variable=value_struct_var}
   > RigVMVariableNode(Value=@n1.curve_struct) {variable=curve_struct_var}
   > SequenceExecution_2 = RigVMAggregateNode()
   n2 = &RigVMFunction_Sequence(ExecuteContext=@SequenceExecution_2.A)
   DeformRig_GetDeltaRefTransform_FNC = RigVMFunctionReferenceNode(Item=(Type=Bone,Name="lowerarm_l"), ExecuteContext=@n2.A)
   > RigVMVariableNode(Value=&RigVMFunction_MathFloatRemap(SourceMaximum=90, SourceMinimum=15, bClamp=true, Value=@DeformRig_GetDeltaRefTransform_FNC.OutTransform.Rotation.Yaw).Result) {variable=l_elbow_alpha}
   DeformRig_GetDeltaRefTransform_FNC_1 = RigVMFunctionReferenceNode(Item=(Type=Bone,Name="lowerarm_r"), ExecuteContext=@n2.B)
   > RigVMVariableNode(Value=&RigVMFunction_MathFloatRemap(SourceMaximum=90, SourceMinimum=15, bClamp=true, Value=@DeformRig_GetDeltaRefTransform_FNC_1.OutTransform.Rotation.Yaw).Result) {variable=r_elbow_alpha}
   RigVMFunction_ControlFlowBranch = &RigVMFunction_ControlFlowBranch(ExecuteContext=@SequenceExecution_2.B, Condition=@n1.elbow)
   SequenceExecution = RigVMAggregateNode(ExecuteContext=@RigVMFunction_ControlFlowBranch.True)
   SequenceExecution_1 = RigVMAggregateNode(ExecuteContext=@SequenceExecution.A)
   &RigUnit_ParentConstraint(
    AdvancedSettings=(RotationOrderForFilter=XZY)
    Child=(Type=Bone,Name="deform_elbow_in_base_l")
    Parents=((Item=(Type=Bone,Name="lowerarm_l"),Weight=0.4),(Item=(Type=Bone,Name="upperarm_l"),Weight=0.6))
    Weight=1
    ExecutePin=@SequenceExecution_1.A
   )
   n3 = RigVMVariableNode() {variable=value_struct_var}
   > &RigUnit_ModifyTransforms(
    ItemToModify=((Item=(Type=Bone,Name="deform_elbow_in_l"),Transform=(Rotation=(X=0,Y=0,Z=0,W=1),Translation=(X=0,Y=0,Z=0),Scale3D=(X=1,Y=1,Z=1))))
    ItemToModify.0.Transform=@n3.Value.elbow_in_24_#25
    Weight=RigVMVariableNode() {variable=l_elbow_alpha}.Value
   )
   &RigUnit_ParentConstraint(
    AdvancedSettings=(RotationOrderForFilter=XZY)
    Child=(Type=Bone,Name="deform_elbow_in_base_r")
    Parents=((Item=(Type=Bone,Name="lowerarm_r"),Weight=0.4),(Item=(Type=Bone,Name="upperarm_r"),Weight=0.6))
    Weight=1
    ExecutePin=@SequenceExecution_1.B
   )
   > DeformRig_GetMirrorTransform_FNC_1 = RigVMFunctionReferenceNode(transform=@n3.Value.elbow_in_24_#25)
   > &RigUnit_ModifyTransforms(
    ItemToModify=((Item=(Type=Bone,Name="deform_elbow_in_r"),Transform=(Rotation=(X=0,Y=0,Z=0,W=1),Translation=(X=0,Y=0,Z=0),Scale3D=(X=1,Y=1,Z=1))))
    ItemToModify.0.Transform=@DeformRig_GetMirrorTransform_FNC_1.mirrorTransform
    Weight=RigVMVariableNode() {variable=r_elbow_alpha}.Value
   )
   SequenceExecution_1_1 = RigVMAggregateNode(ExecuteContext=@SequenceExecution.B)
   &RigUnit_ParentConstraint(
    AdvancedSettings=(RotationOrderForFilter=XZY)
    Child=(Type=Bone,Name="deform_elbow_out_base_l")
    Parents=((Item=(Type=Bone,Name="lowerarm_l"),Weight=0.3),(Item=(Type=Bone,Name="upperarm_l"),Weight=0.7))
    Weight=1
    ExecutePin=@SequenceExecution_1_1.A
   )
   n4 = RigVMVariableNode() {variable=value_struct_var}
   > &RigUnit_ModifyTransforms(
    ItemToModify=((Item=(Type=Bone,Name="deform_elbow_out_l"),Transform=(Rotation=(X=0,Y=0,Z=0,W=1),Translation=(X=0,Y=0,Z=0),Scale3D=(X=1,Y=1,Z=1))))
    ItemToModify.0.Transform=@n4.Value.elbow_out_26_#26
    Weight=RigVMVariableNode() {variable=l_elbow_alpha}.Value
   )
   &RigUnit_ParentConstraint(
    AdvancedSettings=(RotationOrderForFilter=XZY)
    Child=(Type=Bone,Name="deform_elbow_out_base_r")
    Parents=((Item=(Type=Bone,Name="lowerarm_r"),Weight=0.3),(Item=(Type=Bone,Name="upperarm_r"),Weight=0.7))
    Weight=1
    ExecutePin=@SequenceExecution_1_1.B
   )
   > DeformRig_GetMirrorTransform_FNC_2 = RigVMFunctionReferenceNode(transform=@n4.Value.elbow_out_26_#26)
   > &RigUnit_ModifyTransforms(
    ItemToModify=((Item=(Type=Bone,Name="deform_elbow_out_r"),Transform=(Rotation=(X=0,Y=0,Z=0,W=1),Translation=(X=0,Y=0,Z=0),Scale3D=(X=1,Y=1,Z=1))))
    ItemToModify.0.Transform=@DeformRig_GetMirrorTransform_FNC_2.mirrorTransform
    Weight=RigVMVariableNode() {variable=r_elbow_alpha}.Value
   )
   RigVMFunction_Sequence = &RigVMFunction_Sequence(ExecuteContext=@SequenceExecution_2.D)
   n5 = RigVMVariableNode() {variable=curve_struct_var}
   n6 = &RigVMFunction_ControlFlowBranch(ExecuteContext=@RigVMFunction_Sequence.A, Condition=@n5.Value.deformRig_l_elbow_alpha_86_#27)
   RigVMDispatchNode(BoneName=root, Name=deformRig_l_elbow_alpha, ExecuteContext=@n6.True, Value=RigVMVariableNode() {variable=l_elbow_alpha}.Value) {template="DISPATCH_RigDispatch_SetAnimAttribute(in Name,in BoneName,in Value,out Success)"}
   n7 = &RigVMFunction_ControlFlowBranch(ExecuteContext=@RigVMFunction_Sequence.B, Condition=@n5.Value.deformRig_r_elbow_alpha_87_#28)
   RigVMDispatchNode(BoneName=root, Name=deformRig_r_elbow_alpha, ExecuteContext=@n7.True, Value=RigVMVariableNode() {variable=r_elbow_alpha}.Value) {template="DISPATCH_RigDispatch_SetAnimAttribute(in Name,in BoneName,in Value,out Success)"}
   RigVMFunctionReturnNode(ExecuteContext=@SequenceExecution_2.E)
   n8 = &RigVMFunction_ControlFlowBranch(ExecuteContext=@SequenceExecution_2.C, Condition=@n1.bicep)
   SequenceExecution_1_1_1 = RigVMAggregateNode(ExecuteContext=@n8.True)
   n9 = RigVMVariableNode() {variable=value_struct_var}
   &RigUnit_ModifyTransforms(
    ItemToModify=((Item=(Type=Bone,Name="deform_bicep_l"),Transform=(Rotation=(X=0,Y=0,Z=0,W=1),Translation=(X=0,Y=0,Z=0),Scale3D=(X=1,Y=1,Z=1))))
    ItemToModify.0.Transform=@n9.Value.bicep_44_#29
    Weight=RigVMVariableNode() {variable=l_elbow_alpha}.Value
    ExecutePin=@SequenceExecution_1_1_1.A
   )
   DeformRig_GetMirrorTransform_FNC = RigVMFunctionReferenceNode(transform=@n9.Value.bicep_44_#29, ExecuteContext=@SequenceExecution_1_1_1.B)
   > &RigUnit_ModifyTransforms(
    ItemToModify=((Item=(Type=Bone,Name="deform_bicep_r"),Transform=(Rotation=(X=0,Y=0,Z=0,W=1),Translation=(X=0,Y=0,Z=0),Scale3D=(X=1,Y=1,Z=1))))
    ItemToModify.0.Transform=@DeformRig_GetMirrorTransform_FNC.mirrorTransform
    Weight=RigVMVariableNode() {variable=r_elbow_alpha}.Value
   )
  ]}
 ContainedGraph {nodes [
   n1 = RigVMFunctionEntryNode()
   SequenceExecution_2 = &RigVMFunction_Sequence(ExecuteContext=@n1.ExecuteContext)
   SequenceExecution_2_1 = &RigVMFunction_Sequence(ExecuteContext=@SequenceExecution_2.B)
   SequenceExecution_2_2 = &RigVMFunction_Sequence(ExecuteContext=@SequenceExecution_2_1.B)
   SequenceExecution_2_3 = &RigVMFunction_Sequence(ExecuteContext=@SequenceExecution_2_2.B)
   RigVMFunctionReturnNode(A=@SequenceExecution_2.A, B=@SequenceExecution_2_1.A, C=@SequenceExecution_2_2.A, D=@SequenceExecution_2_3.A, E=@SequenceExecution_2_3.B)
  ]}
 ContainedGraph {nodes [
   n1 = RigVMFunctionEntryNode()
   SequenceExecution = &RigVMFunction_Sequence(ExecuteContext=@n1.ExecuteContext)
   SequenceExecution_1 = &RigVMFunction_Sequence(ExecuteContext=@SequenceExecution.B)
   SequenceExecution_2 = &RigVMFunction_Sequence(ExecuteContext=@SequenceExecution_1.B)
   RigVMFunctionReturnNode(A=@SequenceExecution.A, B=@SequenceExecution_1.A, C=@SequenceExecution_2.A, D=@SequenceExecution_2.B)
  ]}
 ContainedGraph {nodes [
   n1 = RigVMFunctionEntryNode()
   SequenceExecution_1 = &RigVMFunction_Sequence(ExecuteContext=@n1.ExecuteContext)
   SequenceExecution_1_1 = &RigVMFunction_Sequence(ExecuteContext=@SequenceExecution_1.B)
   SequenceExecution_1_2 = &RigVMFunction_Sequence(ExecuteContext=@SequenceExecution_1_1.B)
   RigVMFunctionReturnNode(A=@SequenceExecution_1.A, B=@SequenceExecution_1_1.A, C=@SequenceExecution_1_2.A, D=@SequenceExecution_1_2.B)
  ]}
 ContainedGraph {nodes [
   n1 = RigVMFunctionEntryNode()
   SequenceExecution_1_1 = &RigVMFunction_Sequence(ExecuteContext=@n1.ExecuteContext)
   SequenceExecution_1_1_1 = &RigVMFunction_Sequence(ExecuteContext=@SequenceExecution_1_1.B)
   SequenceExecution_1_1_2 = &RigVMFunction_Sequence(ExecuteContext=@SequenceExecution_1_1_1.B)
   RigVMFunctionReturnNode(A=@SequenceExecution_1_1.A, B=@SequenceExecution_1_1_1.A, C=@SequenceExecution_1_1_2.A, D=@SequenceExecution_1_1_2.B)
  ]}
 ContainedGraph {nodes [
   n1 = RigVMFunctionEntryNode()
   SequenceExecution_1_1_1 = &RigVMFunction_Sequence(ExecuteContext=@n1.ExecuteContext)
   SequenceExecution_1_1_1_1 = &RigVMFunction_Sequence(ExecuteContext=@SequenceExecution_1_1_1.B)
   SequenceExecution_1_1_1_2 = &RigVMFunction_Sequence(ExecuteContext=@SequenceExecution_1_1_1_1.B)
   RigVMFunctionReturnNode(A=@SequenceExecution_1_1_1.A, B=@SequenceExecution_1_1_1_1.A, C=@SequenceExecution_1_1_1_2.A, D=@SequenceExecution_1_1_1_2.B)
  ]}
 ContainedGraph {nodes [
   n1 = RigVMFunctionEntryNode()
   > RigVMVariableNode(Value=@n1.value_struct) {variable=value_struct_var}
   > RigVMVariableNode(Value=@n1.curve_struct) {variable=curve_struct_var}
   > SequenceExecution = RigVMAggregateNode()
   n2 = &RigUnit_SphericalPoseReader(
    ActiveRegionScaleFactors=(PositiveWidth=4,NegativeWidth=0.5,PositiveHeight=1.5,NegativeHeight=4)
    ActiveRegionSize=0.05
    Debug=(bDrawDebug=false,DebugScale=25,DebugSegments=20,DebugThickness=0.25)
    DriverAxis=(X=1,Y=0,Z=0)
    DriverItem=(Type=Bone,Name="upperarm_l")
    FalloffRegionScaleFactors=(PositiveWidth=1,NegativeWidth=1,PositiveHeight=1,NegativeHeight=1)
    FalloffSize=0.2
    OptionalParentItem=(Type=Bone,Name="spine_03")
    RotationOffset=(X=0,Y=-150,Z=0)
    ExecutePin=@SequenceExecution.A
    Debug.bDrawDebug=@n1.debug
   )
   n3 = RigVMVariableNode() {variable=value_struct_var}
   > n4 = &RigUnit_ModifyTransforms(
    ItemToModify=((Item=(Type=Bone,Name="deform_lat_l"),Transform=(Rotation=(X=0,Y=0,Z=0,W=1),Translation=(X=0,Y=0,Z=0),Scale3D=(X=1,Y=1,Z=1))))
    ItemToModify.0.Transform=@n3.Value.lat_45_#30
    Weight=@n2.OutputParam
   )
   n5 = RigVMVariableNode() {variable=curve_struct_var}
   RigVMFunction_ControlFlowBranch = &RigVMFunction_ControlFlowBranch(ExecuteContext=@n4.ExecutePin, Condition=@n5.Value.deformRig_l_lat_alpha_91_#31)
   RigVMDispatchNode(BoneName=root, Name=deformRig_l_lat_alpha, ExecuteContext=@RigVMFunction_ControlFlowBranch.True, Value=@n2.OutputParam) {template="DISPATCH_RigDispatch_SetAnimAttribute(in Name,in BoneName,in Value,out Success)"}
   n6 = &RigUnit_SphericalPoseReader(
    ActiveRegionScaleFactors=(PositiveWidth=4,NegativeWidth=0.5,PositiveHeight=1.5,NegativeHeight=4)
    ActiveRegionSize=0.05
    Debug=(bDrawDebug=false,DebugScale=25,DebugSegments=20,DebugThickness=0.25)
    DriverAxis=(X=-1,Y=0,Z=0)
    DriverItem=(Type=Bone,Name="upperarm_r")
    FalloffRegionScaleFactors=(PositiveWidth=1,NegativeWidth=1,PositiveHeight=1,NegativeHeight=1)
    FalloffSize=0.2
    FlipHeightScaling=true
    OptionalParentItem=(Type=Bone,Name="spine_03")
    RotationOffset=(X=0,Y=30,Z=0)
    ExecutePin=@SequenceExecution.B
    Debug.bDrawDebug=@n1.debug
   )
   > DeformRig_GetMirrorTransform_FNC = RigVMFunctionReferenceNode(transform=@n3.Value.lat_45_#30)
   > n7 = &RigUnit_ModifyTransforms(
    ItemToModify=((Item=(Type=Bone,Name="deform_lat_r"),Transform=(Rotation=(X=0,Y=0,Z=0,W=1),Translation=(X=0,Y=0,Z=0),Scale3D=(X=1,Y=1,Z=1))))
    ItemToModify.0.Transform=@DeformRig_GetMirrorTransform_FNC.mirrorTransform
    Weight=@n6.OutputParam
   )
   n8 = &RigVMFunction_ControlFlowBranch(ExecuteContext=@n7.ExecutePin, Condition=@n5.Value.deformRig_r_lat_alpha_93_#32)
   RigVMDispatchNode(BoneName=root, Name=deformRig_r_lat_alpha, ExecuteContext=@n8.True, Value=@n6.OutputParam) {template="DISPATCH_RigDispatch_SetAnimAttribute(in Name,in BoneName,in Value,out Success)"}
   RigVMFunctionReturnNode(ExecuteContext=@SequenceExecution.C)
  ]}
 ContainedGraph {nodes [
   n1 = RigVMFunctionEntryNode()
   SequenceExecution = &RigVMFunction_Sequence(ExecuteContext=@n1.ExecuteContext)
   SequenceExecution_1 = &RigVMFunction_Sequence(ExecuteContext=@SequenceExecution.B)
   SequenceExecution_2 = &RigVMFunction_Sequence(ExecuteContext=@SequenceExecution_1.B)
   RigVMFunctionReturnNode(A=@SequenceExecution.A, B=@SequenceExecution_1.A, C=@SequenceExecution_2.A, D=@SequenceExecution_2.B)
  ]}
 ContainedGraph {nodes [
   n1 = RigVMFunctionEntryNode()
   > RigVMVariableNode(Value=@n1.value_struct) {variable=value_struct_var}
   > RigVMVariableNode(Value=@n1.curve_struct) {variable=curve_struct_var}
   > SequenceExecution = RigVMAggregateNode()
   RigVMFunctionReturnNode(ExecuteContext=@SequenceExecution.D)
   n2 = /Script/ControlRig.RigUnit_AimBoneMath(
    DebugSettings=(Scale=10,WorldOffset=(Rotation=(X=0,Y=0,Z=0,W=1),Translation=(X=0,Y=0,Z=0),Scale3D=(X=1,Y=1,Z=1)))
    Primary=(Weight=1,Axis=(X=-1,Y=0,Z=0),Target=(X=0,Y=0,Z=0),Kind=Location,Space=(Type=Bone,Name="upperarm_r"))
    Secondary=(Weight=1,Axis=(X=0,Y=1,Z=0),Target=(X=0,Y=0,Z=0),Kind=Location,Space=(Type=Bone,Name="spine_04"))
    Weight=1
    InputTransform=&RigUnit_GetTransform(Item=(Type=Bone,Name="deform_pec_aim_r")).Transform
   ) {is=struct RigVMUnitNode}
   &RigUnit_SetTransform(Item=(Type=Bone,Name="deform_pec_aim_r"), Weight=1, bInitial=false, Value=@n2.Result, ExecutePin=@SequenceExecution.B)
   > /Script/ControlRig.RigUnit_PositionConstraintLocalSpaceOffset(
    Child=(Type=Bone,Name="deform_pec_r")
    Filter=(bX=True,bY=false,bZ=false)
    Parents=((Item=(Type=Bone,Name="deform_pec_base_r"),Weight=0),(Item=(Type=Bone,Name="upperarm_r"),Weight=1))
    Weight=1
    bMaintainOffset=True
   ) {is=struct RigVMUnitNode}
   > DeformRig_GetDeltaRefTransform_FNC_1 = RigVMFunctionReferenceNode(Item=(Type=Bone,Name="clavicle_r"))
   n3 = RigVMVariableNode() {variable=value_struct_var}
   > DeformRig_GetMirrorTransform_FNC = RigVMFunctionReferenceNode(transform=@n3.Value.pec_51_#33)
   n4 = &RigVMFunction_MathFloatRemap(SourceMaximum=30, bClamp=true, Value=@DeformRig_GetDeltaRefTransform_FNC_1.OutTransform.Rotation.Pitch)
   > n5 = &RigUnit_ModifyTransforms(
    ItemToModify=((Item=(Type=Bone,Name="deform_pec_r"),Transform=(Rotation=(X=0,Y=0,Z=0,W=1),Translation=(X=0,Y=0,Z=0),Scale3D=(X=1,Y=1,Z=1))))
    ItemToModify.0.Transform=@DeformRig_GetMirrorTransform_FNC.mirrorTransform
    Weight=@n4.Result
   )
   n6 = RigVMVariableNode() {variable=curve_struct_var}
   n7 = &RigVMFunction_ControlFlowBranch(ExecuteContext=@n5.ExecutePin, Condition=@n6.Value.deformRig_r_pec_alpha_85_#34)
   RigVMDispatchNode(BoneName=root, Name=deformRig_r_pec_alpha, ExecuteContext=@n7.True, Value=@n4.Result) {template="DISPATCH_RigDispatch_SetAnimAttribute(in Name,in BoneName,in Value,out Success)"}
   n8 = &RigUnit_AimBoneMath(
    DebugSettings=(Scale=10,WorldOffset=(Rotation=(X=0,Y=0,Z=0,W=1),Translation=(X=0,Y=0,Z=0),Scale3D=(X=1,Y=1,Z=1)))
    Primary=(Weight=1,Axis=(X=1,Y=0,Z=0),Target=(X=0,Y=0,Z=0),Kind=Location,Space=(Type=Bone,Name="upperarm_l"))
    Secondary=(Weight=1,Axis=(X=0,Y=-1,Z=0),Target=(X=0,Y=0,Z=0),Kind=Location,Space=(Type=Bone,Name="spine_04"))
    Weight=1
    InputTransform=&RigUnit_GetTransform(Item=(Type=Bone,Name="deform_pec_aim_l")).Transform
   )
   &RigUnit_SetTransform(Item=(Type=Bone,Name="deform_pec_aim_l"), Weight=1, bInitial=false, Value=@n8.Result, ExecutePin=@SequenceExecution.A)
   > &RigUnit_PositionConstraintLocalSpaceOffset(
    Child=(Type=Bone,Name="deform_pec_l")
    Filter=(bX=True,bY=false,bZ=false)
    Parents=((Item=(Type=Bone,Name="deform_pec_base_l"),Weight=0),(Item=(Type=Bone,Name="upperarm_l"),Weight=1))
    Weight=1
    bMaintainOffset=True
   )
   > DeformRig_GetDeltaRefTransform_FNC = RigVMFunctionReferenceNode(Item=(Type=Bone,Name="clavicle_l"))
   n9 = &RigVMFunction_MathFloatRemap(SourceMaximum=30, bClamp=true, Value=@DeformRig_GetDeltaRefTransform_FNC.OutTransform.Rotation.Pitch)
   > n10 = &RigUnit_ModifyTransforms(
    ItemToModify=((Item=(Type=Bone,Name="deform_pec_l"),Transform=(Rotation=(X=0,Y=0,Z=0,W=1),Translation=(X=0,Y=0,Z=0),Scale3D=(X=1,Y=1,Z=1))))
    ItemToModify.0.Transform=@n3.Value.pec_51_#33
    Weight=@n9.Result
   )
   RigVMFunction_ControlFlowBranch = &RigVMFunction_ControlFlowBranch(ExecuteContext=@n10.ExecutePin, Condition=@n6.Value.deformRig_l_pec_alpha_84_#35)
   DISPATCH_RigDispatch_SetAnimAttribute = RigVMDispatchNode(BoneName=root, Name=deformRig_l_pec_alpha, ExecuteContext=@RigVMFunction_ControlFlowBranch.True, Value=@n9.Result) {template="DISPATCH_RigDispatch_SetAnimAttribute(in Name,in BoneName,in Value,out Success)"}
   &RigUnit_ParentConstraint(
    AdvancedSettings=(RotationOrderForFilter=XZY)
    Child=(Type=Bone,Name="deform_pec_aim_base_l")
    ExecutePin=()
    Parents=((Item=(Type=Bone,Name="spine_03"),Weight=0.25),(Item=(Type=Bone,Name="spine_04"),Weight=0.55),(Item=(Type=Bone,Name="spine_05"),Weight=0.2))
    Weight=1
   )
   &RigUnit_ParentConstraint(
    AdvancedSettings=(RotationOrderForFilter=XZY)
    Child=(Type=Bone,Name="deform_pec_aim_base_r")
    ExecutePin=()
    Parents=((Item=(Type=Bone,Name="spine_03"),Weight=0.25),(Item=(Type=Bone,Name="spine_04"),Weight=0.55),(Item=(Type=Bone,Name="spine_05"),Weight=0.2))
    Weight=1
   )
  ]}
 ContainedGraph {nodes [
   n1 = RigVMFunctionEntryNode()
   SequenceExecution = &RigVMFunction_Sequence(ExecuteContext=@n1.ExecuteContext)
   SequenceExecution_1 = &RigVMFunction_Sequence(ExecuteContext=@SequenceExecution.B)
   SequenceExecution_2 = &RigVMFunction_Sequence(ExecuteContext=@SequenceExecution_1.B)
   RigVMFunctionReturnNode(A=@SequenceExecution.A, B=@SequenceExecution_1.A, C=@SequenceExecution_2.A, D=@SequenceExecution_2.B)
  ]}
 ContainedGraph {nodes [
   n1 = RigVMFunctionEntryNode()
   n2 = &RigVMFunction_MathTransformToEulerTransform(Value=@n1.transform) {template="Cast::Execute(in Value,out Result)"}
   > RigVMFunctionReturnNode(
    mirrorTransform=(Rotation=(X=0,Y=0,Z=0,W=1),Translation=(X=0,Y=0,Z=0),Scale3D=(X=0,Y=0,Z=0))
    mirrorTransform.Rotation=/Script/RigVM.RigVMFunction_MathQuaternionFromRotatorV2(Value=@n2.Result.Rotation) {is=struct RigVMUnitNode, template="Cast::Execute(in Value,out Result)"}.Result
    mirrorTransform.Translation=/Script/RigVM.RigVMFunction_MathVectorScale(Factor=-1, Value=@n2.Result.Location) {is=struct RigVMUnitNode, template="Scale::Execute(in Value,in Factor,out Result)"}.Result
    mirrorTransform.Scale3D=@n2.Result.Scale
   )
  ]}
 ContainedGraph {nodes [
   n1 = RigVMFunctionEntryNode()
   SequenceAggregate_1 = &RigVMFunction_Sequence(ExecuteContext=@n1.ExecuteContext)
   SequenceAggregate_1_1 = &RigVMFunction_Sequence(ExecuteContext=@SequenceAggregate_1.B)
   SequenceAggregate_1_2 = &RigVMFunction_Sequence(ExecuteContext=@SequenceAggregate_1_1.B)
   RigVMFunctionReturnNode(A=@SequenceAggregate_1.A, B=@SequenceAggregate_1_1.A, C=@SequenceAggregate_1_2.A, D=@SequenceAggregate_1_2.B)
  ]}
 ContainedGraph {nodes [
   n1 = RigVMFunctionEntryNode()
   SequenceAggregate_2 = &RigVMFunction_Sequence(ExecuteContext=@n1.ExecuteContext)
   SequenceAggregate_2_1 = &RigVMFunction_Sequence(ExecuteContext=@SequenceAggregate_2.B)
   SequenceAggregate_2_2 = &RigVMFunction_Sequence(ExecuteContext=@SequenceAggregate_2_1.B)
   RigVMFunctionReturnNode(A=@SequenceAggregate_2.A, B=@SequenceAggregate_2_1.A, C=@SequenceAggregate_2_2.A, D=@SequenceAggregate_2_2.B)
  ]}
 New Function_ContainedGraph {nodes [
   n1 = RigVMFunctionEntryNode()
   > /Script/FortniteGame.RigUnit_FortDeformRig(
    Bool=(bUpperArm=False,bElbow=False,bWrist=False,bGroin=False,bGlute=False,bKnee=False,bPec=False,bBicep=False,bLat=False)
    Curve="(bLUpperArmUpAlpha=False,bRUpperArmUpAlpha=False,bLUpperArmFrAlpha=False,bRUpperArmFrAlpha=False,bLUpperArmBkAlpha=False,bRUpperArmBkAlpha=False,bLWristUpAlpha=False,bRWristUpAlpha=False,bLWristDnAlpha=False,bRWristDnAlpha=False,bLKneeAlpha=False,bRKneeAlp ...(truncated)"
    Debug=(bUpperArmDebug=False,bWristDebug=False,bLatDebug=False,bGroinGluteDebug=False)
    Value="(UpperArmUp=(Rotation=(X=0,Y=0,Z=0,W=1),Translation=(X=0,Y=0,Z=0),Scale3D=(X=1,Y=1,Z=1)),UpperArmFwdFr=(Rotation=(X=0,Y=0,Z=0,W=1),Translation=(X=0,Y=0,Z=0),Scale3D=(X=1,Y=1,Z=1)),UpperArmFwdBk=(Rotation=(X=0,Y=0,Z=0,W=1),Translation=(X=0,Y=0,Z=0),Scale3D= ...(truncated)"
    Curve.bLElbowAlpha=@n1.DeformRig_ Curve.deformRig_l_elbow_alpha_86_#27
    Curve.bLKneeAlpha=@n1.DeformRig_ Curve.deformRig_l_knee_alpha_82_#23
    Curve.bLLatAlpha=@n1.DeformRig_ Curve.deformRig_l_lat_alpha_91_#31
    Curve.bLPecAlpha=@n1.DeformRig_ Curve.deformRig_l_pec_alpha_84_#35
    Curve.bLThighUpAlpha=@n1.DeformRig_ Curve.deformRig_l_thigh_up_alpha_88_#19
    Curve.bLUpperArmBkAlpha=@n1.DeformRig_ Curve.deformRig_l_upperarm_bk_alpha_76_#6
    Curve.bLUpperArmFrAlpha=@n1.DeformRig_ Curve.deformRig_l_upperarm_fr_alpha_74_#3
    Curve.bLUpperArmUpAlpha=@n1.DeformRig_ Curve.deformRig_l_upperarm_up_alpha_72_#9
    Curve.bLWristDnAlpha=@n1.DeformRig_ Curve.deformRig_l_wrist_dn_alpha_80_#15
    Curve.bLWristUpAlpha=@n1.DeformRig_ Curve.deformRig_l_wrist_up_alpha_78_#12
    Curve.bRElbowAlpha=@n1.DeformRig_ Curve.deformRig_r_elbow_alpha_87_#28
    Curve.bRKneeAlpha=@n1.DeformRig_ Curve.deformRig_r_knee_alpha_83_#24
    Curve.bRLatAlpha=@n1.DeformRig_ Curve.deformRig_r_lat_alpha_93_#32
    Curve.bRPecAlpha=@n1.DeformRig_ Curve.deformRig_r_pec_alpha_85_#34
    Curve.bRThighUpAlpha=@n1.DeformRig_ Curve.deformRig_r_thigh_up_alpha_89_#20
    Curve.bRUpperArmBkAlpha=@n1.DeformRig_ Curve.deformRig_r_upperarm_bk_alpha_77_#7
    Curve.bRUpperArmFrAlpha=@n1.DeformRig_ Curve.deformRig_r_upperarm_fr_alpha_75_#4
    Curve.bRUpperArmUpAlpha=@n1.DeformRig_ Curve.deformRig_r_upperarm_up_alpha_73_#10
    Curve.bRWristDnAlpha=@n1.DeformRig_ Curve.deformRig_r_wrist_dn_alpha_81_#16
    Curve.bRWristUpAlpha=@n1.DeformRig_ Curve.deformRig_r_wrist_up_alpha_79_#13
    Bool.bBicep=@n1.DeformRig_Bool.bicep_13_#36
    Bool.bElbow=@n1.DeformRig_Bool.elbow_3_#37
    Bool.bGlute=@n1.DeformRig_Bool.glute_17_#38
    Bool.bGroin=@n1.DeformRig_Bool.groin_8_#39
    Bool.bKnee=@n1.DeformRig_Bool.knee_9_#40
    Bool.bLat=@n1.DeformRig_Bool.lat_15_#41
    Bool.bPec=@n1.DeformRig_Bool.pec_12_#42
    Bool.bUpperArm=@n1.DeformRig_Bool.upperarm_1_#43
    Bool.bWrist=@n1.DeformRig_Bool.wrist_7_#44
    Debug.bGroinGluteDebug=@n1.DeformRig_Debug.groinGlute_debug_27_#39
    Debug.bLatDebug=@n1.DeformRig_Debug.lat_debug_26_#41
    Debug.bUpperArmDebug=@n1.DeformRig_Debug.upperarm_debug_18_#43
    Debug.bWristDebug=@n1.DeformRig_Debug.wrist_debug_20_#44
    Value.Bicep=@n1.DeformRig_Value.bicep_44_#29
    Value.ElbowIn=@n1.DeformRig_Value.elbow_in_24_#25
    Value.ElbowOut=@n1.DeformRig_Value.elbow_out_26_#26
    Value.Glute=@n1.DeformRig_Value.glute_36_#17
    Value.Groin=@n1.DeformRig_Value.groin_35_#18
    Value.KneeIn=@n1.DeformRig_Value.knee_in_41_#21
    Value.KneeOut=@n1.DeformRig_Value.knee_out_42_#22
    Value.Lat=@n1.DeformRig_Value.lat_45_#30
    Value.Pec=@n1.DeformRig_Value.pec_51_#33
    Value.UpperArmBk=@n1.DeformRig_Value.upperarm_bk_22_#5
    Value.UpperArmFwdBk=@n1.DeformRig_Value.upperarm_fwd_bk_20_#2
    Value.UpperArmFwdFr=@n1.DeformRig_Value.upperarm_fwd_fr_19_#1
    Value.UpperArmUp=@n1.DeformRig_Value.upperarm_up_18_#8
    Value.WristDown=@n1.DeformRig_Value.wrist_dn_32_#14
    Value.WristUp=@n1.DeformRig_Value.wrist_up_31_#11
   ) {is=struct RigVMUnitNode}
   > RigVMFunctionReturnNode()
   RigVMVariableNode(
    ExecuteContext=()
    Value="(upperarm_1_#43=True,elbow_3_#37=True,wrist_7_#44=True,groin_8_#39=True,glute_17_#38=True,knee_9_40C408324CA261 ...(truncated)"
   ) {variable=deform_bool_struct}
   > RigVMVariableNode(Value="(upperarm_up_18_#8=(Rotation=(X=0,Y=0,Z=0,W=1),Translation=(X=0,Y=0,Z=8),Scale3D=(X=0.7,Y=1,Z=1)),upperarm_fwd_fr_19_#1=(Rotation=(X=0,Y=0,Z=0,W=1),Translation=(X=0,Y=10,Z=0),Scale3D=(X=0.5,Y=1,Z= ...(truncated)") {variable=deform_value_struct}
   > RigVMVariableNode(Value=()) {variable=deform_curve_struct}
   > RigVMVariableNode(Value=(groinGlute_debug_27_#39=False)) {variable=deform_debug_struct}
   > SequenceAggregate = RigVMAggregateNode()
   SequenceAggregate_1 = RigVMAggregateNode(ExecuteContext=@SequenceAggregate.A)
   n2 = &RigVMFunction_ControlFlowBranch(
    Condition=RigVMVariableNode() {variable=deform_bool_struct}.Value.upperarm_1_#43
    ExecuteContext=@SequenceAggregate_1.A
   )
   DeformRig_UpperArm_FNC = RigVMFunctionReferenceNode(
    ExecuteContext=@n2.True
    value_struct=RigVMVariableNode() {variable=deform_value_struct}.Value
    debug=RigVMVariableNode() {variable=deform_debug_struct}.Value.upperarm_debug_18_#43
    curve_struct=RigVMVariableNode() {variable=deform_curve_struct}.Value
   )
   n3 = &RigVMFunction_ControlFlowBranch(
    Condition=RigVMVariableNode() {variable=deform_bool_struct}.Value.wrist_7_#44
    ExecuteContext=@SequenceAggregate_1.B
   )
   DeformRig_Wrist_FNC = RigVMFunctionReferenceNode(
    ExecuteContext=@n3.True
    value_struct=RigVMVariableNode() {variable=deform_value_struct}.Value
    debug=RigVMVariableNode() {variable=deform_debug_struct}.Value.wrist_debug_20_#44
    curve_struct=RigVMVariableNode() {variable=deform_curve_struct}.Value
   )
   n4 = &RigVMFunction_ControlFlowBranch(
    Condition=RigVMVariableNode() {variable=deform_bool_struct}.Value.knee_9_#40
    ExecuteContext=@SequenceAggregate_1.C
   )
   DeformRig_Knee_FNC = RigVMFunctionReferenceNode(
    ExecuteContext=@n4.True
    value_struct=RigVMVariableNode() {variable=deform_value_struct}.Value
    curve_struct=RigVMVariableNode() {variable=deform_curve_struct}.Value
   )
   n5 = &RigVMFunction_ControlFlowBranch(
    Condition=RigVMVariableNode() {variable=deform_bool_struct}.Value.pec_12_#42
    ExecuteContext=@SequenceAggregate_1.D
   )
   DeformRig_Pec_FNC = RigVMFunctionReferenceNode(
    ExecuteContext=@n5.True
    value_struct=RigVMVariableNode() {variable=deform_value_struct}.Value
    curve_struct=RigVMVariableNode() {variable=deform_curve_struct}.Value
   )
   SequenceAggregate_2 = RigVMAggregateNode(ExecuteContext=@SequenceAggregate.B)
   n6 = &RigVMFunction_ControlFlowBranch(
    Condition=RigVMVariableNode() {variable=deform_bool_struct}.Value.lat_15_#41
    ExecuteContext=@SequenceAggregate_2.A
   )
   DeformRig_Lat_FNC = RigVMFunctionReferenceNode(
    ExecuteContext=@n6.True
    value_struct=RigVMVariableNode() {variable=deform_value_struct}.Value
    debug=RigVMVariableNode() {variable=deform_debug_struct}.Value.lat_debug_26_#41
    curve_struct=RigVMVariableNode() {variable=deform_curve_struct}.Value
   )
   n7 = RigVMVariableNode() {variable=deform_bool_struct}
   n8 = &RigVMFunction_ControlFlowBranch(
    Condition=&RigVMFunction_MathBoolOr(B=@n7.Value.bicep_13_#36, A=@n7.Value.elbow_3_#37).Result
    ExecuteContext=@SequenceAggregate_2.B
   )
   DeformRig_Elbow_FNC = RigVMFunctionReferenceNode(
    ExecuteContext=@n8.True
    value_struct=RigVMVariableNode() {variable=deform_value_struct}.Value
    curve_struct=RigVMVariableNode() {variable=deform_curve_struct}.Value
    bicep=@n7.Value.bicep_13_#36
    elbow=@n7.Value.elbow_3_#37
   )
   n9 = RigVMVariableNode() {variable=deform_bool_struct}
   n10 = &RigVMFunction_ControlFlowBranch(
    Condition=&RigVMFunction_MathBoolOr(B=@n9.Value.glute_17_#38, A=@n9.Value.groin_8_#39).Result
    ExecuteContext=@SequenceAggregate_2.C
   )
   DeformRig_GroinGlue_FNC = RigVMFunctionReferenceNode(
    ExecuteContext=@n10.True
    value_struct=RigVMVariableNode() {variable=deform_value_struct}.Value
    debug=RigVMVariableNode() {variable=deform_debug_struct}.Value.groinGlute_debug_27_#39
    curve_struct=RigVMVariableNode() {variable=deform_curve_struct}.Value
    glute=@n9.Value.glute_17_#38
    groin=@n9.Value.groin_8_#39
   )
  ]}
 RigVMFunctionLibrary {nodes [
   DeformRig_Elbow_FNC = RigVMCollapseNode(
    ExecuteContext=()
    bicep=true
    curve_struct=(deformRig_l_lat_alpha_91_#31=False,deformRig_r_lat_alpha_93_#32=False)
    elbow=true
    value_struct="(upperarm_up_18_#8=(Rotation=(X=0,Y=0,Z=0,W=1),Translation=(X=0,Y=0,Z=8),Scale3D=(X=0.7,Y=1,Z=1)),upperarm_fwd_fr_19_#1=(Rotation=(X=0,Y=0,Z=0,W=1),Translation=(X=0,Y=10,Z=0),Scale3D=(X=0.5,Y=1,Z= ...(truncated)"
   )
   DeformRig_GetDeltaRefTransform_FNC = RigVMCollapseNode(ExecuteContext=(), Item=(Type=Bone))
   DeformRig_GetMirrorTransform_FNC = RigVMCollapseNode(ExecuteContext=(), transform=(Rotation=(X=0,Y=0,Z=0,W=1),Translation=(X=0,Y=0,Z=0),Scale3D=(X=1,Y=1,Z=1)))
   DeformRig_GroinGlue_FNC = RigVMCollapseNode(
    ExecuteContext=()
    curve_struct=(deformRig_l_lat_alpha_91_#31=False,deformRig_r_lat_alpha_93_#32=False)
    debug=False
    glute=true
    groin=true
    value_struct="(upperarm_up_18_#8=(Rotation=(X=0,Y=0,Z=0,W=1),Translation=(X=0,Y=0,Z=8),Scale3D=(X=0.7,Y=1,Z=1)),upperarm_fwd_fr_19_#1=(Rotation=(X=0,Y=0,Z=0,W=1),Translation=(X=0,Y=10,Z=0),Scale3D=(X=0.5,Y=1,Z= ...(truncated)"
   )
   DeformRig_Knee_FNC = RigVMCollapseNode(
    ExecuteContext=()
    curve_struct=(deformRig_l_lat_alpha_91_#31=False,deformRig_r_lat_alpha_93_#32=False)
    value_struct="(upperarm_up_18_#8=(Rotation=(X=0,Y=0,Z=0,W=1),Translation=(X=0,Y=0,Z=8),Scale3D=(X=0.7,Y=1,Z=1)),upperarm_fwd_fr_19_#1=(Rotation=(X=0,Y=0,Z=0,W=1),Translation=(X=0,Y=10,Z=0),Scale3D=(X=0.5,Y=1,Z= ...(truncated)"
   )
   DeformRig_Lat_FNC = RigVMCollapseNode(
    ExecuteContext=()
    curve_struct=(deformRig_l_lat_alpha_91_#31=False,deformRig_r_lat_alpha_93_#32=False)
    debug=False
    value_struct="(upperarm_up_18_#8=(Rotation=(X=0,Y=0,Z=0,W=1),Translation=(X=0,Y=0,Z=8),Scale3D=(X=0.7,Y=1,Z=1)),upperarm_fwd_fr_19_#1=(Rotation=(X=0,Y=0,Z=0,W=1),Translation=(X=0,Y=10,Z=0),Scale3D=(X=0.5,Y=1,Z= ...(truncated)"
   )
   DeformRig_Pec_FNC = RigVMCollapseNode(
    ExecuteContext=()
    curve_struct=()
    value_struct="(upperarm_up_18_#8=(Rotation=(X=0,Y=0,Z=0,W=1),Translation=(X=0,Y=0,Z=8),Scale3D=(X=0.7,Y=1,Z=1)),upperarm_fwd_fr_19_#1=(Rotation=(X=0,Y=0,Z=0,W=1),Translation=(X=0,Y=10,Z=0),Scale3D=(X=0.5,Y=1,Z= ...(truncated)"
   )
   DeformRig_UpperArm_FNC = RigVMCollapseNode(
    ExecuteContext=()
    curve_struct=(deformRig_l_lat_alpha_91_#31=False,deformRig_r_lat_alpha_93_#32=False)
    debug=False
    value_struct="(upperarm_up_18_#8=(Rotation=(X=0,Y=0,Z=0,W=1),Translation=(X=0,Y=0,Z=8),Scale3D=(X=0.7,Y=1,Z=1)),upperarm_fwd_fr_19_#1=(Rotation=(X=0,Y=0,Z=0,W=1),Translation=(X=0,Y=10,Z=0),Scale3D=(X=0.5,Y=1,Z= ...(truncated)"
   )
   DeformRig_Wrist_FNC = RigVMCollapseNode(
    ExecuteContext=()
    curve_struct=(deformRig_l_lat_alpha_91_#31=False,deformRig_r_lat_alpha_93_#32=False)
    debug=false
    value_struct="(upperarm_up_18_#8=(Rotation=(X=0,Y=0,Z=0,W=1),Translation=(X=0,Y=0,Z=8),Scale3D=(X=0.7,Y=1,Z=1)),upperarm_fwd_fr_19_#1=(Rotation=(X=0,Y=0,Z=0,W=1),Translation=(X=0,Y=10,Z=0),Scale3D=(X=0.5,Y=1,Z= ...(truncated)"
   )
   FN_DeformRig_FNC = RigVMCollapseNode(
    DeformRig_ Curve=()
    DeformRig_Bool="(upperarm_1_#43=True,elbow_3_#37=True,wrist_7_#44=True,groin_8_#39=True,glute_17_#38=True,knee_9_40C408324CA261 ...(truncated)"
    DeformRig_Debug=(upperarm_debug_18_#43=False,wrist_debug_20_#44=False,groinGlute_debug_27_#39=False,lat_debug_26_#41=False)
    DeformRig_Value="(upperarm_up_18_#8=(Rotation=(X=0,Y=0,Z=0,W=1),Translation=(X=0,Y=0,Z=8),Scale3D=(X=0.7,Y=1,Z=1)),upperarm_fwd_fr_19_#1=(Rotation=(X=0,Y=0,Z=0,W=1),Translation=(X=0,Y=10,Z=0),Scale3D=(X=0.5,Y=1,Z= ...(truncated)"
    ExecuteContext=()
   )
  ]}
 RigVMModel {nodes [RigUnit_BeginExecution = &RigUnit_BeginExecution()]}
 SequenceAggregate_3_ContainedGraph {nodes [
   n1 = RigVMFunctionEntryNode()
   SequenceAggregate = &RigVMFunction_Sequence(ExecuteContext=@n1.ExecuteContext)
   SequenceAggregate_1 = &RigVMFunction_Sequence(ExecuteContext=@SequenceAggregate.B)
   RigVMFunctionReturnNode(A=@SequenceAggregate.A, B=@SequenceAggregate_1.A, C=@SequenceAggregate_1.B)
  ]}
]
